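# Sound cue for the Athena stink grenade burst, built from close, mid and distant layers mixed together, each with its own attenuation and sound class override.
# Close and mid layers each blend random burst and glass variants with modulators, both drawing on the close glass recordings.
# The distant layer plays random pitched-down bursts behind a short delay.
# A concurrency override caps simultaneous instances.
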SoundCue /Game/Athena/Sounds/Weapons/Stink_Grenade/Stink_Grenade_Explode_Cue
root=Mixer_5
nodes [
 {id=Attenuation_4, children [Mixer_0], properties {attenuationSettings=/Game/Sounds/Attenuation/Fort_Explosion_Medium_Athena_Att}}
 {
  id=Attenuation_6
  children [Mixer_2]
  properties {
   attenuationSettings=/Game/Sounds/Attenuation/Fort_Explosion_Close_Athena_Att
   attenuationOverrides {falloffDistance=900, attenuationShapeExtents {x=100}}
  }
 }
 {
  id=Attenuation_8
  children [Modulator_7]
  properties {
   attenuationSettings=/Game/Sounds/Attenuation/Fort_Explosion_Distant_Athena_Att
   attenuationOverrides {falloffDistance=3000, attenuationShapeExtents {x=1000}}
  }
 }
 {id=Delay_0, children [Attenuation_8], properties {delayMax=0.1}}
 {id=Mixer_0, children [Modulator_5,Modulator_0], properties {inputVolume [1,0.8]}}
 {id=Mixer_2, children [Modulator_3,Modulator_4], properties {inputVolume [1,1.3]}}
 {id=Mixer_5, children [SoundClass_3,SoundClass_2,SoundClass_0], properties {inputVolume [0.5,0.35,0.35]}}
 {id=Modulator_0, children [Random_3], properties {pitchMax=1.1, volumeMax=1, volumeMin=1}}
 {id=Modulator_3, children [Random_1], properties {pitchMax=1.1, pitchMin=0.9, volumeMax=1, volumeMin=1}}
 {id=Modulator_4, children [Random_2], properties {pitchMax=1.1, volumeMax=1, volumeMin=1}}
 {id=Modulator_5, children [Random_4]}
 {id=Modulator_7, children [Random_5], properties {pitchMax=0.9, pitchMin=0.8}}
 {id=Random_1, children [WavePlayer_3, WavePlayer_4, WavePlayer_11], properties {weights [1,1,1]}}
 {id=Random_2, children [WavePlayer_5,WavePlayer_6,WavePlayer_7,WavePlayer_8,WavePlayer_9], properties {weights [1,1,1,1,1]}}
 {id=Random_3, children [WavePlayer_10, WavePlayer_13, WavePlayer_14, WavePlayer_15, WavePlayer_16], properties {weights [1,1,1,1,1]}}
 {id=Random_4, children [WavePlayer_12, WavePlayer_19, WavePlayer_20], properties {weights [1,1,1]}}
 {id=Random_5, children [WavePlayer_21, WavePlayer_22, WavePlayer_23], properties {weights [1,1,1]}}
 {id=SoundClass_0, children [Delay_0], properties {soundClassOverride=&Fort_WeaponRanged_Fire_Distant}}
 {
  id=SoundClass_2
  children [Attenuation_4]
  properties {soundClassOverride=&Fort_WeaponRanged_Fire_Distant}
 }
 {
  id=SoundClass_3
  children [Attenuation_6]
  properties {soundClassOverride=&Fort_WeaponRanged_Fire_3P_PSM}
 }
 {id=WavePlayer_10, properties {soundWaveAssetPtr=&/Stink_Grenade_AOE_Burst_Close_Glass_01}}
 {id=WavePlayer_11, properties {soundWaveAssetPtr=&/Stink_Grenade_AOE_Burst_Close_03}}
 {id=WavePlayer_12, properties {soundWaveAssetPtr=&/Stink_Grenade_AOE_Burst_Mid_01}}
 {id=WavePlayer_13, properties {soundWaveAssetPtr=&/Stink_Grenade_AOE_Burst_Close_Glass_02}}
 {id=WavePlayer_14, properties {soundWaveAssetPtr=&/Stink_Grenade_AOE_Burst_Close_Glass_03}}
 {id=WavePlayer_15, properties {soundWaveAssetPtr=&/Stink_Grenade_AOE_Burst_Close_Glass_04}}
 {id=WavePlayer_16, properties {soundWaveAssetPtr=&/Stink_Grenade_AOE_Burst_Close_Glass_05}}
 {id=WavePlayer_19, properties {soundWaveAssetPtr=&/Stink_Grenade_AOE_Burst_Mid_02}}
 {id=WavePlayer_20, properties {soundWaveAssetPtr=&/Stink_Grenade_AOE_Burst_Mid_03}}
 {id=WavePlayer_21, properties {soundWaveAssetPtr=&/Stink_Grenade_AOE_Burst_Distant_01}}
 {id=WavePlayer_22, properties {soundWaveAssetPtr=&/Stink_Grenade_AOE_Burst_Distant_02}}
 {id=WavePlayer_23, properties {soundWaveAssetPtr=&/Stink_Grenade_AOE_Burst_Distant_03}}
 {id=WavePlayer_3, properties {soundWaveAssetPtr=&/Stink_Grenade_AOE_Burst_Close_01}}
 {id=WavePlayer_4, properties {soundWaveAssetPtr=&/Stink_Grenade_AOE_Burst_Close_02}}
 {id=WavePlayer_5, properties {soundWaveAssetPtr=&Stink_Grenade_AOE_Burst_Close_Glass_01}}
 {id=WavePlayer_6, properties {soundWaveAssetPtr=&Stink_Grenade_AOE_Burst_Close_Glass_02}}
 {id=WavePlayer_7, properties {soundWaveAssetPtr=&Stink_Grenade_AOE_Burst_Close_Glass_03}}
 {id=WavePlayer_8, properties {soundWaveAssetPtr=&Stink_Grenade_AOE_Burst_Close_Glass_04}}
 {id=WavePlayer_9, properties {soundWaveAssetPtr=&Stink_Grenade_AOE_Burst_Close_Glass_05}}
]
properties {bOverrideConcurrency=true, maxDistance=27500, volumeMultiplier=1, concurrencyOverrides {maxCount=4}}
