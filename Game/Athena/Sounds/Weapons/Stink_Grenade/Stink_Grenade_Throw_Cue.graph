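# Sound cue for the whoosh played when a Stink Grenade is thrown.
# A Random node picks one of three whoosh waves with equal weight and feeds a Modulator at the root.
# It uses gadget-throw attenuation and the gadget items sound class at slightly reduced volume.
# Its concurrency override prevents new instances from starting once the limit is reached.
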SoundCue /Game/Athena/Sounds/Weapons/Stink_Grenade/Stink_Grenade_Throw_Cue
root=Modulator_0
nodes [
 {id=Modulator_0, children [Random_1]}
 {id=Random_1, children [WavePlayer_3,WavePlayer_4,WavePlayer_5], properties {weights [1,1,1]}}
 {id=WavePlayer_3, properties {soundWaveAssetPtr=&/Stink_Grenade_Throw_Whoosh_01}}
 {id=WavePlayer_4, properties {soundWaveAssetPtr=&/Stink_Grenade_Throw_Whoosh_02}}
 {id=WavePlayer_5, properties {soundWaveAssetPtr=&/Stink_Grenade_Throw_Whoosh_03}}
]
properties {
 attenuationSettings=/Game/Sounds/Attenuation/Fort_GadgetThrow_Att
 bOverrideConcurrency=true
 maxDistance=1536
 soundClassObject=&Fort_Gadgets_Items
 volumeMultiplier=0.8
 concurrencyOverrides {resolutionRule=PreventNew}
}
